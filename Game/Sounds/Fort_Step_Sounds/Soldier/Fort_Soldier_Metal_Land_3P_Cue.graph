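# Sound cue for a third-person Soldier landing on metal, routed to the enemy player foley sound class.
# A random node picks one of six equally weighted metal landing waves, then a modulator varies its pitch and volume slightly.
# The cue plays at reduced volume through Athena third-person land attenuation and references a local footstep concurrency set, alongside override settings that prevent new sounds past a small limit.
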SoundCue /Game/Sounds/Fort_Step_Sounds/Soldier/Fort_Soldier_Metal_Land_3P_Cue
root=Modulator_5
nodes [
 {id=Modulator_5, children [Random_3], properties {pitchMax=1.1, pitchMin=0.9, volumeMax=1.3, volumeMin=0.7}}
 {
  id=Random_3
  children [WavePlayer_34, WavePlayer_35, WavePlayer_36, WavePlayer_37, WavePlayer_38, WavePlayer_39]
  properties {weights [1,1,1,1,1,1]}
 }
 {id=WavePlayer_34, properties {soundWaveAssetPtr=&/Soldier_Step_Land_Metal_01}}
 {id=WavePlayer_35, properties {soundWaveAssetPtr=&/Soldier_Step_Land_Metal_02}}
 {id=WavePlayer_36, properties {soundWaveAssetPtr=&/Soldier_Step_Land_Metal_03}}
 {id=WavePlayer_37, properties {soundWaveAssetPtr=&/Soldier_Step_Land_Metal_04}}
 {id=WavePlayer_38, properties {soundWaveAssetPtr=&/Soldier_Step_Land_Metal_05}}
 {id=WavePlayer_39, properties {soundWaveAssetPtr=&/Soldier_Step_Land_Metal_06}}
]
properties {
 attenuationSettings=/Game/Sounds/Attenuation/Fort_Player_Land_3P_Athena_Att
 maxDistance=3584
 soundClassObject=&Fort_Player_Foley_3P_Enemy_Athena
 volumeMultiplier=0.7
 concurrencyOverrides {maxCount=4, resolutionRule=PreventNew}
 concurrencySet [&Athena_Footstep_Local_Conc]
}
