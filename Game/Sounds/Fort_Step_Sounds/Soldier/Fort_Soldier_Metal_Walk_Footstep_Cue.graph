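# Sound cue for Soldier walk footsteps on metal, routed to the first-person player foley sound class with first-person walk footstep attenuation.
# A mixer layers a random pick of twelve metal step waves at full input with a random pick of six Commando foley waves at low level.
# Each layer passes through a modulator, with the foley given wider pitch and volume variation.
# A local footstep concurrency set is assigned.
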SoundCue /Game/Sounds/Fort_Step_Sounds/Soldier/Fort_Soldier_Metal_Walk_Footstep_Cue
root=Mixer_1
nodes [
 {id=Mixer_1, children [Modulator_11,Modulator_0], properties {inputVolume [0.1,1]}}
 {id=Modulator_0, children [Random_0]}
 {id=Modulator_11, children [Random_3], properties {pitchMax=1, pitchMin=0.7, volumeMax=1, volumeMin=0.3}}
 {
  id=Random_0
  children [
   WavePlayer_0
   WavePlayer_1
   WavePlayer_2
   WavePlayer_3
   WavePlayer_4
   WavePlayer_5
   WavePlayer_12
   WavePlayer_13
   WavePlayer_14
   WavePlayer_15
   WavePlayer_16
   WavePlayer_17
  ]
  properties {weights [1,1,1,1,1,1,1,1,1,1,1,1]}
 }
 {
  id=Random_3
  children [WavePlayer_6, WavePlayer_7, WavePlayer_8, WavePlayer_9, WavePlayer_10, WavePlayer_11]
  properties {weights [1,1,1,1,1,1]}
 }
 {id=WavePlayer_0, properties {soundWaveAssetPtr=&/Soldier_Step_Walk_Metal_10}}
 {id=WavePlayer_1, properties {soundWaveAssetPtr=&/Soldier_Step_Walk_Metal_11}}
 {id=WavePlayer_10, properties {soundWaveAssetPtr=/Game/Sounds/Fortress_Foley/Commando/Commando_Foley_Stereo_005}}
 {id=WavePlayer_11, properties {soundWaveAssetPtr=/Game/Sounds/Fortress_Foley/Commando/Commando_Foley_Stereo_009}}
 {id=WavePlayer_12, properties {soundWaveAssetPtr=&/Soldier_Step_Walk_Metal_04}}
 {id=WavePlayer_13, properties {soundWaveAssetPtr=&/Soldier_Step_Walk_Metal_05}}
 {id=WavePlayer_14, properties {soundWaveAssetPtr=&/Soldier_Step_Walk_Metal_06}}
 {id=WavePlayer_15, properties {soundWaveAssetPtr=&/Soldier_Step_Walk_Metal_07}}
 {id=WavePlayer_16, properties {soundWaveAssetPtr=&/Soldier_Step_Walk_Metal_08}}
 {id=WavePlayer_17, properties {soundWaveAssetPtr=&/Soldier_Step_Walk_Metal_09}}
 {id=WavePlayer_2, properties {soundWaveAssetPtr=&/Soldier_Step_Walk_Metal_12}}
 {id=WavePlayer_3, properties {soundWaveAssetPtr=&/Soldier_Step_Walk_Metal_01}}
 {id=WavePlayer_4, properties {soundWaveAssetPtr=&/Soldier_Step_Walk_Metal_02}}
 {id=WavePlayer_5, properties {soundWaveAssetPtr=&/Soldier_Step_Walk_Metal_03}}
 {id=WavePlayer_6, properties {soundWaveAssetPtr=/Game/Sounds/Fortress_Foley/Commando/Commando_Foley_Stereo_006}}
 {id=WavePlayer_7, properties {soundWaveAssetPtr=/Game/Sounds/Fortress_Foley/Commando/Commando_Foley_Stereo_008}}
 {id=WavePlayer_8, properties {soundWaveAssetPtr=/Game/Sounds/Fortress_Foley/Commando/Commando_Foley_Stereo_001}}
 {id=WavePlayer_9, properties {soundWaveAssetPtr=/Game/Sounds/Fortress_Foley/Commando/Commando_Foley_Stereo_003}}
]
properties {
 attenuationSettings=/Game/Sounds/Attenuation/Fort_Player_Footsteps_1P_Walk_Athena_Att
 maxDistance=2097152
 priority=0.5
 soundClassObject=&Fort_Player_Foley_1P_Athena
 volumeMultiplier=0.5
 concurrencyOverrides {bLimitToOwner=true, maxCount=1, resolutionRule=StopOldest}
 concurrencySet [&Athena_Footstep_Local_Conc]
}
