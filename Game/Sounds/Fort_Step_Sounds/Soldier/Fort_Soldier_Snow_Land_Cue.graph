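# Sound cue for a Soldier landing on snow, routed to the first-person Athena player foley sound class.
# It mixes two randomized layers: snow landing steps at full mixer input and Commando landing foley at reduced input.
# Each layer has its own pitch and volume modulation, with the foley always pitched up.
# A player landing attenuation asset shapes falloff, and a local footstep concurrency set is assigned.
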SoundCue /Game/Sounds/Fort_Step_Sounds/Soldier/Fort_Soldier_Snow_Land_Cue
root=Mixer_5
nodes [
 {id=Mixer_5, children [Modulator_6,Modulator_5], properties {inputVolume [0.4,1]}}
 {id=Modulator_5, children [Random_3], properties {pitchMax=1.1, pitchMin=0.9, volumeMax=1.3, volumeMin=0.7}}
 {id=Modulator_6, children [Random_1], properties {pitchMax=1.4, pitchMin=1.1, volumeMax=1, volumeMin=0.5}}
 {id=Random_1, children [WavePlayer_338, WavePlayer_339, WavePlayer_341], properties {weights [1,1,1]}}
 {id=Random_3, children [WavePlayer_4,WavePlayer_5,WavePlayer_6,WavePlayer_7,WavePlayer_8,WavePlayer_9], properties {weights [1,1,1,1,1,1]}}
 {id=WavePlayer_338, properties {soundWaveAssetPtr=/Game/Sounds/Fortress_Foley/Commando/Commando_Land_Foley_Stereo_003}}
 {id=WavePlayer_339, properties {soundWaveAssetPtr=/Game/Sounds/Fortress_Foley/Commando/Commando_Land_Foley_Stereo_005}}
 {id=WavePlayer_341, properties {soundWaveAssetPtr=/Game/Sounds/Fortress_Foley/Commando/Commando_Land_Foley_Stereo_004}}
 {id=WavePlayer_4, properties {soundWaveAssetPtr=&/Soldier_Step_Land_Snow_01}}
 {id=WavePlayer_5, properties {soundWaveAssetPtr=&/Soldier_Step_Land_Snow_02}}
 {id=WavePlayer_6, properties {soundWaveAssetPtr=&/Soldier_Step_Land_Snow_03}}
 {id=WavePlayer_7, properties {soundWaveAssetPtr=&/Soldier_Step_Land_Snow_04}}
 {id=WavePlayer_8, properties {soundWaveAssetPtr=&/Soldier_Step_Land_Snow_05}}
 {id=WavePlayer_9, properties {soundWaveAssetPtr=&/Soldier_Step_Land_Snow_06}}
]
properties {
 attenuationSettings=/Game/Sounds/Attenuation/Fort_Player_Land_Athena_Att
 maxDistance=2097152
 soundClassObject=&Fort_Player_Foley_1P_Athena
 volumeMultiplier=1
 concurrencyOverrides {maxCount=4, resolutionRule=PreventNew}
 concurrencySet [&Athena_Footstep_Local_Conc]
}
