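# Sound cue for a Soldier landing on stone as heard in third person, routed to the enemy player foley 3P sound class.
# A Random node picks among seven equally weighted waves, six stone landings and one stone walk step, feeding a Modulator that varies pitch and volume.
# A shared Athena landing attenuation asset handles spatial falloff, and the cue references a local footstep concurrency set.
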SoundCue /Game/Sounds/Fort_Step_Sounds/Soldier/Fort_Soldier_Stone_Land_3P_Cue
root=Modulator_5
nodes [
 {id=Modulator_5, children [Random_3], properties {pitchMax=1.1, pitchMin=0.9, volumeMax=1.3, volumeMin=0.7}}
 {
  id=Random_3
  children [WavePlayer_3,WavePlayer_4,WavePlayer_5,WavePlayer_6,WavePlayer_7,WavePlayer_8,WavePlayer_9]
  properties {weights [1,1,1,1,1,1,1]}
 }
 {id=WavePlayer_3, properties {soundWaveAssetPtr=&/Soldier_Step_Walk_Stone_05}}
 {id=WavePlayer_4, properties {soundWaveAssetPtr=&/Soldier_Step_Land_Stone_01}}
 {id=WavePlayer_5, properties {soundWaveAssetPtr=&/Soldier_Step_Land_Stone_02}}
 {id=WavePlayer_6, properties {soundWaveAssetPtr=&/Soldier_Step_Land_Stone_03}}
 {id=WavePlayer_7, properties {soundWaveAssetPtr=&/Soldier_Step_Land_Stone_04}}
 {id=WavePlayer_8, properties {soundWaveAssetPtr=&/Soldier_Step_Land_Stone_05}}
 {id=WavePlayer_9, properties {soundWaveAssetPtr=&/Soldier_Step_Land_Stone_06}}
]
properties {
 attenuationSettings=/Game/Sounds/Attenuation/Fort_Player_Land_3P_Athena_Att
 maxDistance=3584
 soundClassObject=&Fort_Player_Foley_3P_Enemy_Athena
 volumeMultiplier=0.7
 concurrencyOverrides {maxCount=4, resolutionRule=PreventNew}
 concurrencySet [&Athena_Footstep_Local_Conc]
}
