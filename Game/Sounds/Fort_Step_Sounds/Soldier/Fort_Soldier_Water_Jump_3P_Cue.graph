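# Sound cue for a third-person Soldier water jump, routed to the enemy player foley sound class with third-person landing attenuation.
# It randomly picks one of six equally weighted water landing waves.
# Two stacked modulators raise and vary pitch, the outer also varying volume, around an enveloper that ramps volume up from near silence.
# A concurrency override allows one instance and blocks new plays while it sounds.
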SoundCue /Game/Sounds/Fort_Step_Sounds/Soldier/Fort_Soldier_Water_Jump_3P_Cue
root=Modulator_0
nodes [
 {
  id=Enveloper_1
  children [Modulator_1]
  properties {
   pitchCurve {editorCurveData {keys [{time=0}]}}
   volumeCurve {editorCurveData {keys [<at -0.0002 0.0075>,<at 0.1817 0.0667>,<at 0.3 0.5626>]}}
  }
 }
 {id=Modulator_0, children [Enveloper_1], properties {pitchMax=1.3, pitchMin=1.1, volumeMax=1, volumeMin=0.7}}
 {id=Modulator_1, children [Random_2], properties {pitchMax=1.4, pitchMin=1.2}}
 {id=Random_2, children [WavePlayer_1,WavePlayer_2,WavePlayer_3,WavePlayer_4,WavePlayer_5,WavePlayer_6], properties {weights [1,1,1,1,1,1]}}
 {id=WavePlayer_1, properties {soundWaveAssetPtr=&/Soldier_Step_Land_Water_01}}
 {id=WavePlayer_2, properties {soundWaveAssetPtr=&/Soldier_Step_Land_Water_02}}
 {id=WavePlayer_3, properties {soundWaveAssetPtr=&/Soldier_Step_Land_Water_03}}
 {id=WavePlayer_4, properties {soundWaveAssetPtr=&/Soldier_Step_Land_Water_04}}
 {id=WavePlayer_5, properties {soundWaveAssetPtr=&/Soldier_Step_Land_Water_05}}
 {id=WavePlayer_6, properties {soundWaveAssetPtr=&/Soldier_Step_Land_Water_06}}
]
properties {
 attenuationSettings=/Game/Sounds/Attenuation/Fort_Player_Land_3P_Athena_Att
 bOverrideConcurrency=true
 maxDistance=3584
 soundClassObject=&Fort_Player_Foley_3P_Enemy_Athena
 volumeMultiplier=0.5
 concurrencyOverrides {maxCount=1, resolutionRule=PreventNew}
}
