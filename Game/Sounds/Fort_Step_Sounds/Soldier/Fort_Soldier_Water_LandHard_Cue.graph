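# Sound cue for a Soldier's hard landing in water, part of the soldier step sounds.
# A Random node picks one of six equally weighted water landing waves, then a Modulator varies its pitch and volume.
# It uses the Athena player landing attenuation and first-person player foley sound class.
# A concurrency override caps simultaneous instances and blocks new plays once the limit is reached.
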SoundCue /Game/Sounds/Fort_Step_Sounds/Soldier/Fort_Soldier_Water_LandHard_Cue
root=Modulator_4
nodes [
 {id=Modulator_4, children [Random_0], properties {pitchMax=1.1, pitchMin=0.9, volumeMax=1.3, volumeMin=0.7}}
 {id=Random_0, children [WavePlayer_0,WavePlayer_1,WavePlayer_2,WavePlayer_6,WavePlayer_7,WavePlayer_8], properties {weights [1,1,1,1,1,1]}}
 {id=WavePlayer_0, properties {soundWaveAssetPtr=&/Soldier_Step_Land_Water_02}}
 {id=WavePlayer_1, properties {soundWaveAssetPtr=&/Soldier_Step_Land_Water_01}}
 {id=WavePlayer_2, properties {soundWaveAssetPtr=&/Soldier_Step_Land_Water_03}}
 {id=WavePlayer_6, properties {soundWaveAssetPtr=&/Soldier_Step_Land_Water_04}}
 {id=WavePlayer_7, properties {soundWaveAssetPtr=&/Soldier_Step_Land_Water_05}}
 {id=WavePlayer_8, properties {soundWaveAssetPtr=&/Soldier_Step_Land_Water_06}}
]
properties {
 attenuationSettings=/Game/Sounds/Attenuation/Fort_Player_Land_Athena_Att
 bOverrideConcurrency=true
 maxDistance=2097152
 soundClassObject=&Fort_Player_Foley_1P_Athena
 volumeMultiplier=1
 concurrencyOverrides {maxCount=4, resolutionRule=PreventNew}
}
