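# Sound cue for third-person soldier landings in water, routed to the enemy player foley sound class.
# A Random node picks with equal weight among six water landing wave variations.
# It uses a third-person landing attenuation asset and a reduced volume multiplier.
# A concurrency override caps simultaneous instances and blocks new plays once the limit is reached.
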SoundCue /Game/Sounds/Fort_Step_Sounds/Soldier/Fort_Soldier_Water_Land_3P_Cue
root=Random_2
nodes [
 {id=Random_2, children [WavePlayer_3,WavePlayer_4,WavePlayer_5,WavePlayer_6,WavePlayer_7,WavePlayer_8], properties {weights [1,1,1,1,1,1]}}
 {id=WavePlayer_3, properties {soundWaveAssetPtr=&/Soldier_Step_Land_Water_02}}
 {id=WavePlayer_4, properties {soundWaveAssetPtr=&/Soldier_Step_Land_Water_01}}
 {id=WavePlayer_5, properties {soundWaveAssetPtr=&/Soldier_Step_Land_Water_03}}
 {id=WavePlayer_6, properties {soundWaveAssetPtr=&/Soldier_Step_Land_Water_04}}
 {id=WavePlayer_7, properties {soundWaveAssetPtr=&/Soldier_Step_Land_Water_05}}
 {id=WavePlayer_8, properties {soundWaveAssetPtr=&/Soldier_Step_Land_Water_06}}
]
properties {
 attenuationSettings=/Game/Sounds/Attenuation/Fort_Player_Land_3P_Athena_Att
 bOverrideConcurrency=true
 maxDistance=3584
 soundClassObject=&Fort_Player_Foley_3P_Enemy_Athena
 volumeMultiplier=0.7
 concurrencyOverrides {maxCount=4, resolutionRule=PreventNew}
}
