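# Sound cue for the Soldier's third-person footsteps when walking in water, classed as enemy player foley in Athena.
# A Random node picks among twelve equally weighted water step waves and feeds a Modulator at the root for variation.
# It uses the Athena third-person walk footstep attenuation at reduced volume.
# A concurrency override limits it to one instance per owner, stopping the oldest.
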SoundCue /Game/Sounds/Fort_Step_Sounds/Soldier/Fort_Soldier_Water_Walk_Footstep_3P_Cue
root=Modulator_0
nodes [
 {id=Modulator_0, children [Random_0]}
 {
  id=Random_0
  children [
   WavePlayer_0
   WavePlayer_1
   WavePlayer_2
   WavePlayer_3
   WavePlayer_4
   WavePlayer_5
   WavePlayer_12
   WavePlayer_13
   WavePlayer_14
   WavePlayer_15
   WavePlayer_16
   WavePlayer_17
  ]
  properties {weights [1,1,1,1,1,1,1,1,1,1,1,1]}
 }
 {id=WavePlayer_0, properties {soundWaveAssetPtr=&/Soldier_Step_Walk_Water_12}}
 {id=WavePlayer_1, properties {soundWaveAssetPtr=&/Soldier_Step_Walk_Water_10}}
 {id=WavePlayer_12, properties {soundWaveAssetPtr=&/Soldier_Step_Walk_Water_04}}
 {id=WavePlayer_13, properties {soundWaveAssetPtr=&/Soldier_Step_Walk_Water_05}}
 {id=WavePlayer_14, properties {soundWaveAssetPtr=&/Soldier_Step_Walk_Water_06}}
 {id=WavePlayer_15, properties {soundWaveAssetPtr=&/Soldier_Step_Walk_Water_07}}
 {id=WavePlayer_16, properties {soundWaveAssetPtr=&/Soldier_Step_Walk_Water_08}}
 {id=WavePlayer_17, properties {soundWaveAssetPtr=&/Soldier_Step_Walk_Water_09}}
 {id=WavePlayer_2, properties {soundWaveAssetPtr=&/Soldier_Step_Walk_Water_11}}
 {id=WavePlayer_3, properties {soundWaveAssetPtr=&/Soldier_Step_Walk_Water_01}}
 {id=WavePlayer_4, properties {soundWaveAssetPtr=&/Soldier_Step_Walk_Water_02}}
 {id=WavePlayer_5, properties {soundWaveAssetPtr=&/Soldier_Step_Walk_Water_03}}
]
properties {
 attenuationSettings=/Game/Sounds/Attenuation/Fort_Player_Footsteps_3P_Walk_Athena_Att
 bOverrideConcurrency=true
 maxDistance=3584
 priority=0.5
 soundClassObject=&Fort_Player_Foley_3P_Enemy_Athena
 volumeMultiplier=0.5
 concurrencyOverrides {bLimitToOwner=true, maxCount=1, resolutionRule=StopOldest}
}
