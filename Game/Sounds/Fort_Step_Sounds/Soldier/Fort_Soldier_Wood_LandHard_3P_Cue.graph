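# Sound cue for a Soldier character's hard landing on wood, heard in third person.
# A Random node picks one of six equally weighted landing waves, then a Modulator varies its pitch and volume on each play.
# It uses the Athena third-person landing attenuation, the enemy 3P foley sound class and a footstep concurrency set that caps simultaneous instances.
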SoundCue /Game/Sounds/Fort_Step_Sounds/Soldier/Fort_Soldier_Wood_LandHard_3P_Cue
root=Modulator_4
nodes [
 {id=Modulator_4, children [Random_2], properties {pitchMax=1.1, pitchMin=0.9, volumeMax=1.3, volumeMin=0.7}}
 {
  id=Random_2
  children [WavePlayer_15, WavePlayer_16, WavePlayer_17, WavePlayer_18, WavePlayer_21, WavePlayer_22]
  properties {weights [1,1,1,1,1,1]}
 }
 {id=WavePlayer_15, properties {soundWaveAssetPtr=&/Soldier_Step_Land_Wood_01}}
 {id=WavePlayer_16, properties {soundWaveAssetPtr=&/Soldier_Step_Land_Wood_02}}
 {id=WavePlayer_17, properties {soundWaveAssetPtr=&/Soldier_Step_Land_Wood_03}}
 {id=WavePlayer_18, properties {soundWaveAssetPtr=&/Soldier_Step_Land_Wood_04}}
 {id=WavePlayer_21, properties {soundWaveAssetPtr=&/Soldier_Step_Land_Wood_05}}
 {id=WavePlayer_22, properties {soundWaveAssetPtr=&/Soldier_Step_Land_Wood_06}}
]
properties {
 attenuationSettings=/Game/Sounds/Attenuation/Fort_Player_Land_3P_Athena_Att
 maxDistance=3584
 soundClassObject=&Fort_Player_Foley_3P_Enemy_Athena
 volumeMultiplier=0.8
 concurrencyOverrides {maxCount=4, resolutionRule=PreventNew}
 concurrencySet [&Athena_Footstep_Local_Conc]
}
