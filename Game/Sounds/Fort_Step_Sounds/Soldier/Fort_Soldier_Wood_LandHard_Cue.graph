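# Sound cue for the Soldier's hard landing on wood, routed to the Athena first-person player foley class.
# A mixer layers a random wood landing step at full volume with quieter, pitched-up random Constructor landing foley borrowed for body movement.
# An outer modulator varies pitch and volume on each play.
# Player-landing attenuation and a local footstep concurrency set govern playback.
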SoundCue /Game/Sounds/Fort_Step_Sounds/Soldier/Fort_Soldier_Wood_LandHard_Cue
root=Modulator_4
nodes [
 {id=Mixer_2, children [Modulator_0,Random_2], properties {inputVolume [0.3,1]}}
 {id=Modulator_0, children [Random_0], properties {pitchMax=1.4, pitchMin=1.2}}
 {id=Modulator_4, children [Mixer_2], properties {pitchMax=1.1, pitchMin=0.9, volumeMax=1.3, volumeMin=0.7}}
 {id=Random_0, children [WavePlayer_13, WavePlayer_14, WavePlayer_0, WavePlayer_1], properties {weights [1,1,1,1]}}
 {
  id=Random_2
  children [WavePlayer_15, WavePlayer_16, WavePlayer_17, WavePlayer_18, WavePlayer_21, WavePlayer_22]
  properties {weights [1,1,1,1,1,1]}
 }
 {id=WavePlayer_0, properties {soundWaveAssetPtr=/Game/Sounds/Fortress_Foley/Constructor/Constructor_Land_Foley_Stereo_003}}
 {id=WavePlayer_1, properties {soundWaveAssetPtr=/Game/Sounds/Fortress_Foley/Constructor/Constructor_Land_Foley_Stereo_004}}
 {id=WavePlayer_13, properties {soundWaveAssetPtr=/Game/Sounds/Fortress_Foley/Constructor/Constructor_Land_Foley_Stereo_001}}
 {id=WavePlayer_14, properties {soundWaveAssetPtr=/Game/Sounds/Fortress_Foley/Constructor/Constructor_Land_Foley_Stereo_002}}
 {id=WavePlayer_15, properties {soundWaveAssetPtr=&/Soldier_Step_Land_Wood_01}}
 {id=WavePlayer_16, properties {soundWaveAssetPtr=&/Soldier_Step_Land_Wood_02}}
 {id=WavePlayer_17, properties {soundWaveAssetPtr=&/Soldier_Step_Land_Wood_03}}
 {id=WavePlayer_18, properties {soundWaveAssetPtr=&/Soldier_Step_Land_Wood_04}}
 {id=WavePlayer_21, properties {soundWaveAssetPtr=&/Soldier_Step_Land_Wood_05}}
 {id=WavePlayer_22, properties {soundWaveAssetPtr=&/Soldier_Step_Land_Wood_06}}
]
properties {
 attenuationSettings=/Game/Sounds/Attenuation/Fort_Player_Land_Athena_Att
 maxDistance=2097152
 soundClassObject=&Fort_Player_Foley_1P_Athena
 volumeMultiplier=1
 concurrencyOverrides {maxCount=4, resolutionRule=PreventNew}
 concurrencySet [&Athena_Footstep_Local_Conc]
}
