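# Sound cue for third-person Soldier walking footsteps on wood, routed to the enemy player foley sound class.
# A Random node picks among twelve equally weighted wood step waves, then a Modulator at default settings feeds the output.
# A walk footstep attenuation asset shapes falloff, and concurrency is set through a shared footstep set, with overrides specifying one instance per owner that stops the oldest.
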SoundCue /Game/Sounds/Fort_Step_Sounds/Soldier/Fort_Soldier_Wood_Walk_Footstep_3P_Cue
root=Modulator_0
nodes [
 {id=Modulator_0, children [Random_1]}
 {
  id=Random_1
  children [
   WavePlayer_24
   WavePlayer_25
   WavePlayer_26
   WavePlayer_27
   WavePlayer_28
   WavePlayer_29
   WavePlayer_30
   WavePlayer_31
   WavePlayer_32
   WavePlayer_33
   WavePlayer_34
   WavePlayer_35
  ]
  properties {weights [1,1,1,1,1,1,1,1,1,1,1,1]}
 }
 {id=WavePlayer_24, properties {soundWaveAssetPtr=&/Soldier_Step_Walk_Wood_09}}
 {id=WavePlayer_25, properties {soundWaveAssetPtr=&/Soldier_Step_Walk_Wood_10}}
 {id=WavePlayer_26, properties {soundWaveAssetPtr=&/Soldier_Step_Walk_Wood_11}}
 {id=WavePlayer_27, properties {soundWaveAssetPtr=&/Soldier_Step_Walk_Wood_12}}
 {id=WavePlayer_28, properties {soundWaveAssetPtr=&/Soldier_Step_Walk_Wood_01}}
 {id=WavePlayer_29, properties {soundWaveAssetPtr=&/Soldier_Step_Walk_Wood_02}}
 {id=WavePlayer_30, properties {soundWaveAssetPtr=&/Soldier_Step_Walk_Wood_03}}
 {id=WavePlayer_31, properties {soundWaveAssetPtr=&/Soldier_Step_Walk_Wood_04}}
 {id=WavePlayer_32, properties {soundWaveAssetPtr=&/Soldier_Step_Walk_Wood_05}}
 {id=WavePlayer_33, properties {soundWaveAssetPtr=&/Soldier_Step_Walk_Wood_06}}
 {id=WavePlayer_34, properties {soundWaveAssetPtr=&/Soldier_Step_Walk_Wood_07}}
 {id=WavePlayer_35, properties {soundWaveAssetPtr=&/Soldier_Step_Walk_Wood_08}}
]
properties {
 attenuationSettings=/Game/Sounds/Attenuation/Fort_Player_Footsteps_3P_Walk_Athena_Att
 maxDistance=3584
 priority=0.5
 soundClassObject=&Fort_Player_Foley_3P_Enemy_Athena
 volumeMultiplier=0.5
 concurrencyOverrides {bLimitToOwner=true, maxCount=1, resolutionRule=StopOldest}
 concurrencySet [&Athena_Footstep_Local_Conc]
}
